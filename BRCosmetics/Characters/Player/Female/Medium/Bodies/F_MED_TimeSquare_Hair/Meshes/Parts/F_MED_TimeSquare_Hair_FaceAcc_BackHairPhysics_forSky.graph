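# Physics asset for the back hair of the TimeSquare female medium cosmetic, in a "forSky" variant.
# Three capsule chains (center, left, right) each link four dyn_hair_bk bones to a MSTR root body.
# Head, neck, clavicle and upper-spine capsules provide body shapes around the hair.
# Swing limits are tight at each root and widen toward the tips, while twist stays narrow except at the center tip.
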PhysicsAsset /BRCosmetics/Characters/Player/Female/Medium/Bodies/F_MED_TimeSquare_Hair/Meshes/Parts/F_MED_TimeSquare_Hair_FaceAcc_BackHairPhysics_forSky
bodies {
 clavicle_l [{
   shapeType=Capsule
   length=14.5
   radius=8.7
   shapeName=clavicle_l_capsule
   center=<3.8676 0.3294 -2.3341>
   rotation=<rot 81.119 6.6038 -0.7558>
  }]
 clavicle_r [{
   shapeType=Capsule
   length=15.5
   radius=6.7
   shapeName=clavicle_r_capsule
   center=<-10.8806 0.1029 0.8608>
   rotation=<rot -89.1849 -174.0422 0.0361>
  }]
 dyn_hair_bk_01 [{shapeType=Capsule, length=2.5, radius=5, shapeName=dyn_hair_bk_01_capsule, center=<5 0 0>}]
 dyn_hair_bk_02 [{shapeType=Capsule, length=5.5, radius=7, shapeName=dyn_hair_bk_02_capsule, center=<5 -5.2 0>}]
 dyn_hair_bk_03 [{shapeType=Capsule, length=6, radius=8.5, shapeName=dyn_hair_bk_03_capsule, center=<5.9598 -6 0>}]
 dyn_hair_bk_04 [{shapeType=Capsule, length=6, radius=8.5, shapeName=dyn_hair_bk_04_capsule, center=<9.0432 -6.7455 0>}]
 dyn_hair_bk_MSTR [{shapeType=Capsule, length=1, radius=0.5, shapeName=dyn_hair_bk_MSTR_capsule, rotation=<rot 90 0 0>}]
 dyn_hair_bk_l_01 [{shapeType=Capsule, length=2.5, radius=3.5, shapeName=dyn_hair_bk_l_01_capsule, center=<2.6644 0 0>}]
 dyn_hair_bk_l_02 [{shapeType=Capsule, length=3, radius=4, shapeName=dyn_hair_bk_l_02_capsule, center=<3.216 0 0>}]
 dyn_hair_bk_l_03 [{shapeType=Capsule, length=3, radius=6, shapeName=dyn_hair_bk_l_03_capsule, center=<2.5649 -4 0.22>, rotation=<rot 0 0.0004 11.5505>}]
 dyn_hair_bk_l_04 [{
   shapeType=Capsule
   length=6.5
   radius=6
   shapeName=dyn_hair_bk_l_04_capsule
   center=<6.9458 -5.1 -1.8>
   rotation=<rot 72.6681 18.8646 3.8824>
  }]
 dyn_hair_bk_l_MSTR [{shapeType=Capsule, length=1, radius=0.5, shapeName=dyn_hair_bk_l_MSTR_capsule, rotation=<rot 90 0 0>}]
 dyn_hair_bk_r_01 [{shapeType=Capsule, length=2.5, radius=3.5, shapeName=dyn_hair_bk_r_01_capsule, center=<2.6644 0 0>}]
 dyn_hair_bk_r_02 [{shapeType=Capsule, length=3, radius=4, shapeName=dyn_hair_bk_r_02_capsule, center=<3.216 0 0>}]
 dyn_hair_bk_r_03 [{
   shapeType=Capsule
   length=4.5
   radius=5
   shapeName=dyn_hair_bk_r_03_capsule
   center=<2.6789 -2.2631 -1.024>
   rotation=<rot 82.4967 -19.3701 -7.0514>
  }]
 dyn_hair_bk_r_04 [{shapeType=Capsule, length=9, radius=5.5, shapeName=dyn_hair_bk_r_04_capsule, center=<6.513 -3.85 4>, rotation=<rot 61.6007 -171.954 180>}]
 dyn_hair_bk_r_MSTR [{shapeType=Capsule, length=1, radius=0.5, shapeName=dyn_hair_bk_r_MSTR_capsule, rotation=<rot 90 0 0>}]
 head [{shapeType=Capsule, length=3, radius=7, shapeName=head_capsule, center=<3.4376 0 0>, rotation=<rot 90 0 0>}]
 neck_01 [{shapeType=Capsule, length=13.5, radius=6.5, shapeName=neck_01_capsule, center=<1.8124 -0.0105 0>, rotation=<rot 90 180 -174.0069>}]
 spine_04 [{shapeType=Capsule, length=15, radius=21, shapeName=spine_04_capsule, center=<0.4921 10.6789 0>}]
}
constraints [
 {bone1Name=dyn_hair_bk_04, bone2Name=dyn_hair_bk_03, swing1LimitDegrees=30, swing2LimitDegrees=15, twistLimitDegrees=10}
 {bone1Name=dyn_hair_bk_01, bone2Name=dyn_hair_bk_MSTR, swing1LimitDegrees=5, swing2LimitDegrees=1, twistLimitDegrees=1}
 {bone1Name=dyn_hair_bk_r_04, bone2Name=dyn_hair_bk_r_03, swing1LimitDegrees=30, swing2LimitDegrees=15, twistLimitDegrees=2}
 {bone1Name=dyn_hair_bk_r_01, bone2Name=dyn_hair_bk_r_MSTR, swing1LimitDegrees=5, swing2LimitDegrees=1, twistLimitDegrees=1}
 {bone1Name=dyn_hair_bk_l_04, bone2Name=dyn_hair_bk_l_03, swing1LimitDegrees=20, swing2LimitDegrees=15, twistLimitDegrees=2}
 {bone1Name=dyn_hair_bk_l_01, bone2Name=dyn_hair_bk_l_MSTR, swing1LimitDegrees=5, swing2LimitDegrees=1, twistLimitDegrees=1}
 {bone1Name=dyn_hair_bk_r_02, bone2Name=dyn_hair_bk_r_01, swing1LimitDegrees=15, swing2LimitDegrees=10, twistLimitDegrees=1}
 {bone1Name=dyn_hair_bk_02, bone2Name=dyn_hair_bk_01, swing1LimitDegrees=15, swing2LimitDegrees=5, twistLimitDegrees=1}
 {bone1Name=dyn_hair_bk_l_02, bone2Name=dyn_hair_bk_l_01, swing1LimitDegrees=15, swing2LimitDegrees=10, twistLimitDegrees=1}
 {bone1Name=dyn_hair_bk_03, bone2Name=dyn_hair_bk_02, swing1LimitDegrees=15, swing2LimitDegrees=5, twistLimitDegrees=1}
 {bone1Name=dyn_hair_bk_l_03, bone2Name=dyn_hair_bk_l_02, swing1LimitDegrees=15, swing2LimitDegrees=10, twistLimitDegrees=1}
 {bone1Name=dyn_hair_bk_r_03, bone2Name=dyn_hair_bk_r_02, swing1LimitDegrees=15, swing2LimitDegrees=10, twistLimitDegrees=1}
]
