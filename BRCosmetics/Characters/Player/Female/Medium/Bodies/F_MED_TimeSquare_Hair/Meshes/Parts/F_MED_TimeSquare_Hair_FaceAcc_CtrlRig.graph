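# Control Rig for the TimeSquare female hair part that applies corrective offsets to five hair master bones as the head pose changes.
# Two spherical pose readers, driven by pelvis and right clavicle relative to the head, write HeadBackWeight and HeadRightWeight.
# Each weight blends a chain of Modify Transforms on the back, back-side and side hair masters, head-back first.
# No head-left reader exists, and the head-back reader has debug drawing enabled.
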ControlRigBlueprint /BRCosmetics/Characters/Player/Female/Medium/Bodies/F_MED_TimeSquare_Hair/Meshes/Parts/F_MED_TimeSquare_Hair_FaceAcc_CtrlRig
graphs [
 RigVMFunctionLibrary {links [], nodes []}
 RigVMModel {nodes [
   RigUnit_BeginExecution = &RigUnit_BeginExecution()
   n1 = &RigUnit_SphericalPoseReader(
    ActiveRegionScaleFactors=(PositiveWidth=0.4,NegativeWidth=1.2,PositiveHeight=0.4,NegativeHeight=0.4)
    ActiveRegionSize=0.1
    Debug=(bDrawDebug=true,bDraw2D=False,bDrawLocalAxes=False,DebugScale=25,DebugSegments=20,DebugThickness=0.25)
    DriverAxis=(X=1,Y=0,Z=0)
    DriverItem=(Type=Bone,Name="pelvis")
    FalloffRegionScaleFactors=(PositiveWidth=1,NegativeWidth=1,PositiveHeight=1,NegativeHeight=1)
    FalloffSize=0.22
    FlipHeightScaling=False
    FlipWidthScaling=False
    OptionalParentItem=(Type=Bone,Name="head")
    RotationOffset=(X=100,Y=0,Z=-35)
    ExecutePin=@RigUnit_BeginExecution.ExecutePin
   )
   > RigVMVariableNode(Value=@n1.OutputParam) {variable=HeadBackWeight}
   n2 = RigVMVariableNode() {variable=HeadBackWeight}
   > &RigUnit_ModifyTransforms(
    ItemToModify=((Item=(Type=Bone,Name="dyn_hair_bk_l_MSTR"),Transform=(Rotation=(X=-0.1736,Y=0,Z=0,W=0.9848),Translation=(X=0,Y=0,Z=0),Scale3D=(X=1,Y=1,Z=1))))
    WeightMinimum=0
    Weight=@n2.Value
   )
   > &RigUnit_ModifyTransforms(
    ItemToModify=((Item=(Type=Bone,Name="dyn_hair_bk_MSTR"),Transform=(Rotation=(X=-0.3007,Y=0,Z=0,W=0.9537),Translation=(X=0,Y=0,Z=0),Scale3D=(X=1,Y=1,Z=1))))
    WeightMinimum=0
    Weight=@n2.Value
   )
   > &RigUnit_ModifyTransforms(
    ItemToModify=((Item=(Type=Bone,Name="dyn_hair_bk_r_MSTR"),Transform=(Rotation=(X=-0.1736,Y=0,Z=0,W=0.9848),Translation=(X=0,Y=0,Z=0),Scale3D=(X=1,Y=1,Z=1))))
    WeightMinimum=0
    Weight=@n2.Value
   )
   > &RigUnit_ModifyTransforms(
    ItemToModify=((Item=(Type=Bone,Name="dyn_hair_l_MSTR"),Transform=(Rotation=(X=-0.1736,Y=-0.0041,Z=-0.0627,W=0.9828),Translation=(X=0.5,Y=-2,Z=1),Scale3D=(X=1,Y=1,Z=1))))
    WeightMinimum=0
    Weight=@n2.Value
   )
   > &RigUnit_ModifyTransforms(
    ItemToModify=((Item=(Type=Bone,Name="dyn_hair_r_MSTR"),Transform=(Rotation=(X=-0.1736,Y=0,Z=0,W=0.9848),Translation=(X=0,Y=-1,Z=0),Scale3D=(X=1,Y=1,Z=1))))
    WeightMinimum=0
    Weight=@n2.Value
   )
   > n3 = &RigUnit_SphericalPoseReader(
    ActiveRegionScaleFactors=(PositiveWidth=1,NegativeWidth=1,PositiveHeight=1,NegativeHeight=1)
    ActiveRegionSize=0.02
    Debug=(bDrawDebug=false,bDraw2D=False,bDrawLocalAxes=False,DebugScale=25,DebugSegments=20,DebugThickness=0.25)
    DriverAxis=(X=-1,Y=0,Z=0)
    DriverItem=(Type=Bone,Name="clavicle_r")
    FalloffRegionScaleFactors=(PositiveWidth=1,NegativeWidth=1,PositiveHeight=1,NegativeHeight=1)
    FalloffSize=0.14
    FlipHeightScaling=False
    FlipWidthScaling=False
    OptionalParentItem=(Type=Bone,Name="head")
    RotationOffset=(X=180,Y=-23,Z=0)
   )
   > RigVMVariableNode(Value=@n3.OutputParam) {variable=HeadRightWeight}
   n4 = RigVMVariableNode() {variable=HeadRightWeight}
   > &RigUnit_ModifyTransforms(
    ItemToModify=((Item=(Type=Bone,Name="dyn_hair_bk_l_MSTR"),Transform=(Rotation=(X=0,Y=0.1736,Z=0,W=0.9848),Translation=(X=2,Y=-4,Z=0),Scale3D=(X=1,Y=1,Z=1))))
    WeightMinimum=0
    Weight=@n4.Value
   )
   > &RigUnit_ModifyTransforms(
    ItemToModify=((Item=(Type=Bone,Name="dyn_hair_bk_MSTR"),Transform=(Rotation=(X=0,Y=0.1305,Z=0,W=0.9914),Translation=(X=3,Y=-2,Z=0),Scale3D=(X=1,Y=1,Z=1))))
    WeightMinimum=0
    Weight=@n4.Value
   )
   > &RigUnit_ModifyTransforms(
    ItemToModify=((Item=(Type=Bone,Name="dyn_hair_bk_r_MSTR"),Transform=(Rotation=(X=-0.2504,Y=0,Z=0,W=0.9681),Translation=(X=3,Y=-3,Z=4),Scale3D=(X=1,Y=1,Z=1))))
    WeightMinimum=0
    Weight=@n4.Value
   )
   > &RigUnit_ModifyTransforms(
    ItemToModify=((Item=(Type=Bone,Name="dyn_hair_r_MSTR"),Transform=(Rotation=(X=0,Y=0,Z=0,W=1),Translation=(X=0,Y=-1,Z=0),Scale3D=(X=1,Y=1,Z=1))))
    WeightMinimum=0
    Weight=@n4.Value
   )
   > &RigUnit_ModifyTransforms(
    ItemToModify=((Item=(Type=Bone,Name="dyn_hair_l_MSTR"),Transform=(Rotation=(X=-0.0695,Y=0.0061,Z=-0.0869,W=0.9938),Translation=(X=0.2,Y=-2,Z=0),Scale3D=(X=1,Y=1,Z=1))))
    WeightMinimum=0
    Weight=@n4.Value
   )
  ]}
]
hierarchy [
 root {loc=<0 0 0>, rot=<0 0 0 1>}
 pelvis {parent=root, loc=<0 -0.5301 98.0571>, rot=<0.0127 0.707 -0.0127 -0.707>}
 spine_01 {parent=pelvis, loc=<6.1208 -0.9328 0>, rot=<0 0 0.034 -0.9994>}
 spine_02 {parent=spine_01, loc=<6.5996 -0.2095 0>, rot=<0 0 0.0238 -0.9997>}
 spine_03 {parent=spine_02, loc=<6.4914 -0.1237 0>, rot=<0 0 0.0149 -0.9999>}
 spine_04 {parent=spine_03, loc=<6.9598 0.0794 0>, rot=<0 0 -0.0132 -0.9999>}
 spine_05 {parent=spine_04, loc=<6.9987 0 0>, rot=<0 0 -0.0015 -1>}
 neck_01 {parent=spine_05, loc=<11.3138 0.1529 0>, rot=<0 0 -0.2003 -0.9797>}
 neck_02 {parent=neck_01, loc=<4.5627 0 0>, rot=<0 0 0.0504 -0.9987>}
 head {parent=neck_02, loc=<5.3914 0 0>, rot=<0 0 0.115 -0.9934>}
 dyn_hair_bk_MSTR {parent=head, loc=<11.0745 -6.3352 0>, rot=<0.0029 -0.7071 -0.0029 -0.7071>}
 dyn_hair_bk_01 {parent=dyn_hair_bk_MSTR, loc=<0 0 0>, rot=<0.081 -0.7024 0.081 -0.7024>}
 dyn_hair_bk_02 {parent=dyn_hair_bk_01, loc=<7.8193 0 0>, rot=<0 0 -0.0176 -0.9998>}
 dyn_hair_bk_03 {parent=dyn_hair_bk_02, loc=<6.6762 0 0>, rot=<0 0 0.0739 -0.9973>}
 dyn_hair_bk_04 {parent=dyn_hair_bk_03, loc=<7.4932 0 0>, rot=<0 0 -0.117 -0.9931>}
 dyn_hair_bk_l_MSTR {parent=head, loc=<9.9386 -2.6405 -6.3161>, rot=<0.0029 -0.7071 -0.0029 -0.7071>}
 dyn_hair_bk_l_01 {parent=dyn_hair_bk_l_MSTR, loc=<0 0 0>, rot=<0.0469 -0.6639 0.0526 -0.7445>}
 dyn_hair_bk_l_02 {parent=dyn_hair_bk_l_01, loc=<6.9616 0 0>, rot=<0.002 -0.0148 -0.0071 -0.9999>}
 dyn_hair_bk_l_03 {parent=dyn_hair_bk_l_02, loc=<8.3239 0 0>, rot=<0.0242 -0.0627 0.238 -0.9689>}
 dyn_hair_bk_l_04 {parent=dyn_hair_bk_l_03, loc=<7.0489 0 0>, rot=<0 0 -0.0464 -0.9989>}
 dyn_hair_bk_r_MSTR {parent=head, loc=<9.9332 -1.9809 6.5292>, rot=<0.0029 -0.7071 -0.0029 -0.7071>}
 dyn_hair_bk_r_01 {parent=dyn_hair_bk_r_MSTR, loc=<0 0 0>, rot=<0.0166 -0.7391 0.0151 -0.6732>}
 dyn_hair_bk_r_02 {parent=dyn_hair_bk_r_01, loc=<6.8842 0 0>, rot=<0.0059 -0.044 0.0876 -0.9952>}
 dyn_hair_bk_r_03 {parent=dyn_hair_bk_r_02, loc=<8.1779 0 0>, rot=<-0.0825 0.1482 0.2792 -0.9451>}
 dyn_hair_bk_r_04 {parent=dyn_hair_bk_r_03, loc=<6.6926 0 0>, rot=<-0.0249 -0.1129 -0.1739 -0.9779>}
 dyn_hair_l_MSTR {parent=head, loc=<2.513 1.086 -7.582>, rot=<0.0029 -0.7071 -0.0029 -0.7071>}
 dyn_hair_l_01 {parent=dyn_hair_l_MSTR, loc=<0 0 0>, rot=<0.7071 0 -0.7071 0>}
 dyn_hair_l_02 {parent=dyn_hair_l_01, loc=<4.98 0 0>, rot=<0 0 0.0526 -0.9986>}
 dyn_hair_l_03 {parent=dyn_hair_l_02, loc=<5.1277 0 0>, rot=<0 0 0.1599 -0.9871>}
 dyn_hair_l_04 {parent=dyn_hair_l_03, loc=<4.71 0 0>, rot=<0 0 0.0714 -0.9974>}
 dyn_hair_r_MSTR {parent=head, loc=<2.506 1.941 7.147>, rot=<0.0029 -0.7071 -0.0029 -0.7071>}
 dyn_hair_r_01 {parent=dyn_hair_r_MSTR, loc=<0 0 0>, rot=<0.7071 0 -0.7071 0>}
 dyn_hair_r_02 {parent=dyn_hair_r_01, loc=<4.98 0 0>, rot=<0 0 0.0526 -0.9986>}
 dyn_hair_r_03 {parent=dyn_hair_r_02, loc=<5.128 0 0>, rot=<-0.0511 -0.1206 0.1966 -0.9717>}
 dyn_hair_r_04 {parent=dyn_hair_r_03, loc=<5.538 0 0>, rot=<-0.0054 0.0613 0.0868 -0.9943>}
 clavicle_l {parent=spine_05, loc=<5.4831 -0.1053 -3.2946>, rot=<0.0465 -0.7289 -0.009 -0.6829>}
 upperarm_l {parent=clavicle_l, loc=<9.8658 -1.2136 -0.4105>, rot=<0.0164 -0.3245 -0.0269 -0.9454>}
 clavicle_r {parent=spine_05, loc=<5.4831 -0.1053 3.2946>, rot=<0.7289 0.0465 -0.6829 0.009>}
 upperarm_r {parent=clavicle_r, loc=<-9.8658 1.2136 0.4105>, rot=<0.0164 -0.3245 -0.0269 -0.9454>}
]
variables [
 HeadBackWeight {type=float, default=1}
 HeadRightWeight {type=float, default=1}
]
