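# Physics asset for the front hair of the TimeSquare female medium character, defining capsule bodies and joint limits for two dynamic hair strands.
# Left and right chains each link a MSTR root through four segments with angular-limited constraints, tightest at the root and generally looser farther down.
# Unconstrained capsules on the head, neck, upper spine, clavicles and upper arms approximate the body, apparently as colliders for the hair.
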PhysicsAsset /BRCosmetics/Characters/Player/Female/Medium/Bodies/F_MED_TimeSquare_Hair/Meshes/Parts/F_MED_TimeSquare_Hair_FaceAcc_FrontHairPhysics
bodies {
 clavicle_l [{shapeType=Capsule, length=9, radius=5, shapeName=clavicle_l_capsule, center=<3.8958 -0.1672 -0.3651>, rotation=<rot 80 -0.0018 -12.2871>}]
 clavicle_r [{shapeType=Capsule, length=8, radius=6, shapeName=clavicle_r_capsule, center=<-3.8697 0.8469 0.5132>, rotation=<rot 80 0 0>}]
 dyn_hair_l_01 [{shapeType=Capsule, length=3, radius=2.5, shapeName=dyn_hair_l_01_capsule, center=<3.0918 -0.75 0>, rotation=<rot 0 -0.0003 -59.4863>}]
 dyn_hair_l_02 [{shapeType=Capsule, length=2, radius=2.5, shapeName=dyn_hair_l_02_capsule, center=<3.1957 -0.3737 0.1852>, rotation=<rot 0 -0.0002 -57.7509>}]
 dyn_hair_l_03 [{shapeType=Capsule, length=3, radius=2.5, shapeName=dyn_hair_l_03_capsule, center=<3.9675 -0.4896 1>, rotation=<rot 90 -165.9638 -146.6353>}]
 dyn_hair_l_04 [{shapeType=Capsule, length=6, radius=2.5, shapeName=dyn_hair_l_04_capsule, center=<4.6157 -2.2 1.4>, rotation=<rot 90 178.2101 -169.4817>}]
 dyn_hair_l_MSTR [{shapeType=Capsule, length=1, radius=0.5, shapeName=dyn_hair_l_MSTR_capsule, rotation=<rot 90 0 0>}]
 dyn_hair_r_01 [{shapeType=Capsule, length=2, radius=2.5, shapeName=dyn_hair_r_01_capsule, center=<3 0.24 0.7>, rotation=<rot 0 -0.0005 88.151>}]
 dyn_hair_r_02 [{shapeType=Capsule, length=1.5, radius=2.8, shapeName=dyn_hair_r_02_capsule, center=<3.0918 0.16 0.5>, rotation=<rot 0 -0.0066 58.5044>}]
 dyn_hair_r_03 [{shapeType=Capsule, length=5.5, radius=2, shapeName=dyn_hair_r_03_capsule, center=<3.2978 0 -1.8>, rotation=<rot 83.0471 -21.2879 -9.828>}]
 dyn_hair_r_04 [{
   shapeType=Capsule
   length=3.5
   radius=3.5
   shapeName=dyn_hair_r_04_capsule
   center=<4.7546 -1.8 1.1>
   rotation=<rot 67.5674 173.0676 -179.9999>
  }]
 dyn_hair_r_MSTR [{shapeType=Capsule, length=1, radius=0.5, shapeName=dyn_hair_r_MSTR_capsule, rotation=<rot 90 0 0>}]
 head [{shapeType=Capsule, length=2.5, radius=5.5, shapeName=head_capsule, center=<-0.174 2.8862 0>, rotation=<rot 90 172.875 -171.4604>}]
 neck_01 [{shapeType=Capsule, length=6, radius=4.5, shapeName=neck_01_capsule, center=<1.7875 0.408 0>, rotation=<rot 90 178.2101 -170.5164>}]
 spine_05 [{shapeType=Capsule, length=11, radius=33.5, shapeName=spine_05_capsule, center=<-20.6804 -13.8577 0>, rotation=<rot 90 180 -140.3635>}]
 upperarm_l [{
   shapeType=Capsule
   length=14.5
   radius=4.5
   shapeName=upperarm_l_capsule
   center=<8.5173 -0.1195 0.1462>
   rotation=<rot 82.8138 -0.0022 4.4014>
  }]
 upperarm_r [{
   shapeType=Capsule
   length=23
   radius=3
   shapeName=upperarm_r_capsule
   center=<-11.4913 -0.4316 -1.1384>
   rotation=<rot 86.3895 -0.5037 -8.2584>
  }]
}
constraints [
 {bone1Name=dyn_hair_l_04, bone2Name=dyn_hair_l_03, swing1LimitDegrees=40, swing2LimitDegrees=40, twistLimitDegrees=20}
 {bone1Name=dyn_hair_l_03, bone2Name=dyn_hair_l_02, swing1LimitDegrees=40, swing2LimitDegrees=40, twistLimitDegrees=30}
 {bone1Name=dyn_hair_l_02, bone2Name=dyn_hair_l_01, swing1LimitDegrees=20, swing2LimitDegrees=40, twistLimitDegrees=10}
 {bone1Name=dyn_hair_l_01, bone2Name=dyn_hair_l_MSTR, swing1LimitDegrees=15, swing2LimitDegrees=1, twistLimitDegrees=2}
 {bone1Name=dyn_hair_r_04, bone2Name=dyn_hair_r_03, swing1LimitDegrees=30, swing2LimitDegrees=30, twistLimitDegrees=20}
 {bone1Name=dyn_hair_r_03, bone2Name=dyn_hair_r_02, swing1LimitDegrees=60, swing2LimitDegrees=10, twistLimitDegrees=5}
 {bone1Name=dyn_hair_r_02, bone2Name=dyn_hair_r_01, swing1LimitDegrees=40, swing2LimitDegrees=30, twistLimitDegrees=10}
 {bone1Name=dyn_hair_r_01, bone2Name=dyn_hair_r_MSTR, swing1LimitDegrees=10, swing2LimitDegrees=1, twistLimitDegrees=2}
]
