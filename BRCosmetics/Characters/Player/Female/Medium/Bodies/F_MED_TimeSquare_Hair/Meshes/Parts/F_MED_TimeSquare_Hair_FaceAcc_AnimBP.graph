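# Animation Blueprint that drives secondary hair motion for the TimeSquare female character's hair/face-accessory part.
# It copies the parent mesh pose, adds a Control Rig outside skydiving, and applies bone constraints.
# Cached poses chain back-hair RigidBody, front-hair RigidBody, then spring-driven AnimDynamics curls, with skydiving, dive and parachute flags blending hair offsets and simulations.
# The event graph tracks body-mesh upper-arm rotations, back bling and a gravity override.
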AnimBlueprint /BRCosmetics/Characters/Player/Female/Medium/Bodies/F_MED_TimeSquare_Hair/Meshes/Parts/F_MED_TimeSquare_Hair_FaceAcc_AnimBP
target_skeleton=&/F_MED_TimeSquare_Hair_FaceAcc_Skeleton
graphs [
 AnimGraph {nodes [
   n1 = Anim Dynamics - Bone: dyn_curl_01(ComponentPose=Local To Component(LocalPose=Use cached pose 'RB_Pose'()), Alpha=1, ExternalForce="0,0,0") {
    angularDampingOverride=0.999
    angularSpringConstant=500
    bAngularSpring=true
    bOverrideAngularDamping=true
    bUseGravityOverride=true
    gravityScale=0.35
    simulationSpace=RootRelative
    alphaBoolBlend {blendOption=Linear}
    boundBone {boneName=dyn_curl_01}
    componentAppliedLinearAccClamp=<1000 1000 1000>
    componentLinearAccScale=<0.6 0.6 0.6>
    componentLinearVelScale=<0.5 0.5 0.5>
    gravityOverride {z=-980}
    physicsBodyDefinitions [{
      boundBone {boneName=dyn_curl_01}
      boxExtents=<12 12 12>
      constraintSetup {angularLimitsMax=<5 5 30>, angularLimitsMin=<-5 -40 -30>, angularTarget=<1 0 0>}
      localJointOffset=<1.3 0 0>
     }]
    relativeSpaceBone {boneName=head}
   }
   n2 = Anim Dynamics - Bone: dyn_curl_02(ComponentPose=@n1, Alpha=1, ExternalForce="0,0,0") {
    angularDampingOverride=0.999
    angularSpringConstant=500
    bAngularSpring=true
    bOverrideAngularDamping=true
    bUseGravityOverride=true
    gravityScale=0.35
    simulationSpace=RootRelative
    alphaBoolBlend {blendOption=Linear}
    boundBone {boneName=dyn_curl_02}
    componentAppliedLinearAccClamp=<1000 1000 1000>
    componentLinearAccScale=<0.6 0.6 0.6>
    componentLinearVelScale=<0.5 0.5 0.5>
    gravityOverride {z=-980}
    physicsBodyDefinitions [{
      boundBone {boneName=dyn_curl_02}
      boxExtents=<10 10 10>
      constraintSetup {angularLimitsMax=<5 5 30>, angularLimitsMin=<-5 -30 -30>, angularTarget=<1 0 0>}
      localJointOffset=<1.5 -0.5 0>
     }]
    relativeSpaceBone {boneName=head}
   }
   n3 = Anim Dynamics - Bone: dyn_curl_03(ComponentPose=@n2, Alpha=1, ExternalForce="0,0,0") {
    angularDampingOverride=0.999
    angularSpringConstant=500
    bAngularSpring=true
    bOverrideAngularDamping=true
    bUseGravityOverride=true
    gravityScale=0.35
    simulationSpace=RootRelative
    alphaBoolBlend {blendOption=Linear}
    boundBone {boneName=dyn_curl_03}
    componentAppliedLinearAccClamp=<1000 1000 1000>
    componentLinearAccScale=<0.6 0.6 0.6>
    componentLinearVelScale=<0.5 0.5 0.5>
    gravityOverride {z=-980}
    physicsBodyDefinitions [{
      boundBone {boneName=dyn_curl_03}
      boxExtents=<15 15 15>
      constraintSetup {angularLimitsMax=<5 5 30>, angularLimitsMin=<-5 -30 -30>, angularTarget=<1 0 0>}
      localJointOffset=<1.5 0 0>
     }]
    relativeSpaceBone {boneName=head}
   }
   n4 = Anim Dynamics - Bone: dyn_curl_04(ComponentPose=@n3, Alpha=1, ExternalForce="0,0,0") {
    angularDampingOverride=0.999
    angularSpringConstant=500
    bAngularSpring=true
    bOverrideAngularDamping=true
    bUseGravityOverride=true
    gravityScale=0.35
    simulationSpace=RootRelative
    alphaBoolBlend {blendOption=Linear}
    boundBone {boneName=dyn_curl_04}
    componentAppliedLinearAccClamp=<1000 1000 1000>
    componentLinearAccScale=<0.6 0.6 0.6>
    componentLinearVelScale=<0.5 0.5 0.5>
    gravityOverride {z=-980}
    physicsBodyDefinitions [{
      boundBone {boneName=dyn_curl_04}
      boxExtents=<12 12 12>
      constraintSetup {angularLimitsMax=<5 5 30>, angularLimitsMin=<-5 -20 -40>, angularTarget=<1 0 0>}
      localJointOffset=<1.3 0 0>
     }]
    relativeSpaceBone {boneName=head}
   }
   n5 = Anim Dynamics - Bone: dyn_curl_05(ComponentPose=@n4, Alpha=1, ExternalForce="0,0,0") {
    angularDampingOverride=0.999
    angularSpringConstant=500
    bAngularSpring=true
    bOverrideAngularDamping=true
    bUseGravityOverride=true
    gravityScale=0.35
    simulationSpace=RootRelative
    alphaBoolBlend {blendOption=Linear}
    boundBone {boneName=dyn_curl_05}
    componentAppliedLinearAccClamp=<1000 1000 1000>
    componentLinearAccScale=<0.6 0.6 0.6>
    componentLinearVelScale=<0.5 0.5 0.5>
    gravityOverride {z=-980}
    physicsBodyDefinitions [{
      boundBone {boneName=dyn_curl_05}
      boxExtents=<12 12 12>
      constraintSetup {angularLimitsMax=<5 50 5>, angularLimitsMin=<-5 -20 -20>, angularTarget=<1 0 0>}
      localJointOffset=<1.3 0 0>
     }]
    relativeSpaceBone {boneName=head}
   }
   n6 = Anim Dynamics - Bone: dyn_curl_08(ComponentPose=@n5, Alpha=1, ExternalForce="0,0,0") {
    angularDampingOverride=0.999
    angularSpringConstant=500
    bAngularSpring=true
    bOverrideAngularDamping=true
    bUseGravityOverride=true
    gravityScale=0.35
    simulationSpace=RootRelative
    alphaBoolBlend {blendOption=Linear}
    boundBone {boneName=dyn_curl_08}
    componentAppliedLinearAccClamp=<1000 1000 1000>
    componentLinearAccScale=<0.6 0.6 0.6>
    componentLinearVelScale=<0.5 0.5 0.5>
    gravityOverride {z=-980}
    physicsBodyDefinitions [{
      boundBone {boneName=dyn_curl_08}
      boxExtents=<10 10 10>
      constraintSetup {angularLimitsMax=<5 5 30>, angularLimitsMin=<-5 -50 -30>, angularTarget=<1 0 0>}
      localJointOffset=<1.3 0.5 0>
     }]
    relativeSpaceBone {boneName=head}
   }
   n7 = Anim Dynamics - Bone: dyn_curl_06(ComponentPose=@n6, Alpha=1, ExternalForce="0,0,0") {
    angularDampingOverride=0.999
    angularSpringConstant=3000
    bAngularSpring=true
    bOverrideAngularDamping=true
    bUseGravityOverride=true
    gravityScale=0.35
    simulationSpace=RootRelative
    alphaBoolBlend {blendOption=Linear}
    boundBone {boneName=dyn_curl_06}
    componentAppliedLinearAccClamp=<1000 1000 1000>
    componentLinearAccScale=<0.6 0.6 0.6>
    componentLinearVelScale=<0.5 0.5 0.5>
    gravityOverride {z=-980}
    physicsBodyDefinitions [{
      boundBone {boneName=dyn_curl_06}
      boxExtents=<15 15 15>
      constraintSetup {angularLimitsMax=<2 20 20>, angularLimitsMin=<-2 5 -30>, angularTarget=<1 0 0>}
      localJointOffset=<1 0 0>
     }]
    relativeSpaceBone {boneName=head}
   }
   n8 = Anim Dynamics - Bone: dyn_curl_07(ComponentPose=@n7, Alpha=1, ExternalForce="0,0,0") {
    angularDampingOverride=0.999
    angularSpringConstant=3000
    bAngularSpring=true
    bOverrideAngularDamping=true
    bUseGravityOverride=true
    gravityScale=0.35
    simulationSpace=RootRelative
    alphaBoolBlend {blendOption=Linear}
    boundBone {boneName=dyn_curl_07}
    componentAppliedLinearAccClamp=<1000 1000 1000>
    componentLinearAccScale=<0.6 0.6 0.6>
    componentLinearVelScale=<0.5 0.5 0.5>
    gravityOverride {z=-980}
    physicsBodyDefinitions [{
      boundBone {boneName=dyn_curl_07}
      boxExtents=<15 15 15>
      constraintSetup {angularLimitsMax=<2 20 5>, angularLimitsMin=<-2 -20 -20>, angularTarget=<1 0 0>}
      localJointOffset=<1 0 0>
     }]
    relativeSpaceBone {boneName=head}
   }
   Output Pose(Result=Component To Local(ComponentPose=@n8)) {layerGroup=None}
   n9 = RigidBody!(
    ComponentPose=Local To Component(LocalPose=Use cached pose 'RB_BackPose'())
    bAlphaBoolEnabled:bool
    OverrideWorldGravity:Vector="0,0,-980"
    ComponentLinearAccScale:Vector="0.3,0.3,0.3"
    ComponentLinearVelScale:Vector="0.1,0.1,0.1"
   ) -> (Pose) {
    alphaInputType=Bool
    bClampLinearTranslationLimitToRefPose=true
    bOverrideWorldGravity=true
    overridePhysicsAsset=&/F_MED_TimeSquare_Hair_FaceAcc_FrontHairPhysics
    simulationSpace=BaseBoneSpace
    alphaBoolBlend {blendOption=Linear}
    alphaScaleBiasClamp {interpSpeedDecreasing=4, interpSpeedIncreasing=6}
    baseBoneRef {boneName=spine_05}
    componentAppliedLinearAccClamp=<3000 3000 3000>
    componentLinearAccScale=<0.4 0.4 0.4>
    componentLinearVelScale=<0.3 0.3 0.3>
    overrideWorldGravity {z=-875}
    simSpaceSettings {maxAngularAcceleration=200, maxAngularVelocity=5, maxLinearAcceleration=500, maxLinearVelocity=200}
   }
   n10 = RigidBody(
    ComponentPose=Local To Component(LocalPose=Use cached pose 'RB_BackPose'())
    OverrideWorldGravity="0,0,-980"
    ComponentLinearAccScale="0.3,0.3,0.3"
    ComponentLinearVelScale="0.1,0.1,0.1"
   ) {
    alphaInputType=Bool
    bClampLinearTranslationLimitToRefPose=true
    bOverrideWorldGravity=true
    overridePhysicsAsset=&F_MED_TimeSquare_Hair_FaceAcc_FrontHairPhysics
    simulationSpace=BaseBoneSpace
    alphaBoolBlend {blendOption=Linear}
    alphaScaleBiasClamp {interpSpeedDecreasing=4, interpSpeedIncreasing=6}
    baseBoneRef {boneName=spine_05}
    componentAppliedLinearAccClamp=<3000 3000 3000>
    componentLinearAccScale=<0.4 0.4 0.4>
    componentLinearVelScale=<0.3 0.3 0.3>
    overrideWorldGravity {z=-875}
    simSpaceSettings {maxAngularAcceleration=200, maxAngularVelocity=5, maxLinearAcceleration=500, maxLinearVelocity=200}
   }
   Save cached pose 'RB_Pose'(Pose=Blend Poses by bool(
     BlendPose_0=Component To Local(ComponentPose=@n10)
     BlendPose_1=Component To Local(ComponentPose=@n9)
     BlendTime_0=0.9
     BlendTime_1=0.9
     bActiveValue=Get bIsSkydiveDiveMode()
    ) {blendType=Linear}) {cacheName=RB_Pose}
   n11 = "Transform (Modify) Bone - Bone: dyn_hair_bk_l_MSTR"!(ComponentPose=Local To Component(LocalPose=Use cached pose 'StartPose'()), Alpha=1, Rotation="0,0,-20") -> (Pose) {
    rotationMode=BMM_Additive
    rotationSpace=BCS_BoneSpace
    scaleMode=BMM_Additive
    translationSpace=BCS_WorldSpace
    alphaBoolBlend {blendInTime=1, blendOption=Linear, blendOutTime=1}
    boneToModify {boneName=dyn_hair_bk_l_MSTR}
    rotation {roll=-20}
   }
   n12 = "Transform (Modify) Bone - Bone: dyn_hair_bk_MSTR"!(ComponentPose=@n11, Alpha=1, Rotation="0,0,-10") -> (Pose) {
    rotationMode=BMM_Additive
    rotationSpace=BCS_BoneSpace
    scaleMode=BMM_Additive
    translationSpace=BCS_WorldSpace
    alphaBoolBlend {blendInTime=1, blendOption=Linear, blendOutTime=1}
    boneToModify {boneName=dyn_hair_bk_MSTR}
    rotation {roll=-10}
   }
   n13 = "Transform (Modify) Bone - Bone: dyn_hair_bk_r_MSTR"!(ComponentPose=@n12, Alpha=1, Rotation="0,0,-20") -> (Pose) {
    rotationMode=BMM_Additive
    rotationSpace=BCS_BoneSpace
    scaleMode=BMM_Additive
    translationSpace=BCS_WorldSpace
    alphaBoolBlend {blendInTime=1, blendOption=Linear, blendOutTime=1}
    boneToModify {boneName=dyn_hair_bk_r_MSTR}
    rotation {roll=-20}
   }
   Save cached pose 'DiveModePose'(Pose=Component To Local(ComponentPose=@n13)) {cacheName=DiveModePose}
   n14 = Control Rig(Source=Copy Pose From Mesh() {bUseAttachedParent=true}) {
    alphaBoolBlend {blendOption=Linear}
    controlRigAssetReference {blueprintRigClass=&/F_MED_TimeSquare_Hair_FaceAcc_CtrlRig.F_MED_TimeSquare_Hair_FaceAcc_CtrlRig_C}
   }
   n15 = Copy Pose From Mesh() {bUseAttachedParent=true}
   n16 = Get bIsSkyDiving()
   n17 = "Constraint - dyn_back_curl_03 (PPPPPPP)"!(
    ComponentPose=Local To Component(LocalPose=Blend Poses by bool(BlendPose_0=@n15, BlendPose_1=@n14, bActiveValue=@n16) {blendType=Linear})
    Alpha=1
    ConstraintWeights_0=0.0848
    ConstraintWeights_1:float=0.0877
    ConstraintWeights_2:float=0.284
    ConstraintWeights_3:float=0.0018
    ConstraintWeights_4:float=0.1214
    ConstraintWeights_5:float=0.1626
    ConstraintWeights_6:float=0.2575
   ) -> (Pose) {
    alphaBoolBlend {blendOption=Linear}
    boneToModify {boneName=dyn_back_curl_03}
    constraintSetup [
     {targetBone {boneName=dyn_hair_bk_02}}
     {targetBone {boneName=dyn_hair_bk_03}}
     {targetBone {boneName=dyn_hair_bk_04}}
     {targetBone {boneName=dyn_hair_bk_l_01}}
     {targetBone {boneName=dyn_hair_bk_l_02}}
     {targetBone {boneName=dyn_hair_bk_l_03}}
     {targetBone {boneName=dyn_hair_bk_l_04}}
    ]
    constraintWeights [1,1,1,1,1,1,1]
   }
   n18 = "Constraint - dyn_curl_06 (PPPP)"!(
    ComponentPose=@n17
    Alpha=1
    ConstraintWeights_0=0.142
    ConstraintWeights_1:float=0.026
    ConstraintWeights_2:float=0.596
    ConstraintWeights_3:float=0.236
   ) -> (Pose) {
    alphaBoolBlend {blendOption=Linear}
    boneToModify {boneName=dyn_curl_06}
    constraintSetup [
     {targetBone {boneName=head}}
     {targetBone {boneName=dyn_hair_r_02}}
     {targetBone {boneName=dyn_hair_r_01}}
     {targetBone {boneName=dyn_hair_r_MSTR}}
    ]
    constraintWeights [1,1,1,1]
   }
   n19 = "Constraint - dyn_curl_07 (PPPP)"!(
    ComponentPose=@n18
    Alpha=1
    ConstraintWeights_0=0.02
    ConstraintWeights_1:float=0.21
    ConstraintWeights_2:float=0.605
    ConstraintWeights_3:float=0.165
   ) -> (Pose) {
    alphaBoolBlend {blendOption=Linear}
    boneToModify {boneName=dyn_curl_07}
    constraintSetup [
     {targetBone {boneName=head}}
     {targetBone {boneName=dyn_hair_l_02}}
     {targetBone {boneName=dyn_hair_l_01}}
     {targetBone {boneName=dyn_hair_l_MSTR}}
    ]
    constraintWeights [1,1,1,1]
   }
   n20 = "Transform (Modify) Bone - Bone: upperarm_l"!(ComponentPose=@n19, Alpha=1, Rotation="0,0,0") -> (Pose) {
    rotationMode=BMM_Replace
    rotationSpace=BCS_WorldSpace
    scaleMode=BMM_Additive
    translationSpace=BCS_WorldSpace
    alphaBoolBlend {blendInTime=1, blendOption=Linear, blendOutTime=1}
    boneToModify {boneName=upperarm_l}
   }
   n21 = "Transform (Modify) Bone - Bone: upperarm_r"!(ComponentPose=@n20, Alpha=1, Rotation="0,0,0") -> (Pose) {
    rotationMode=BMM_Replace
    rotationSpace=BCS_WorldSpace
    scaleMode=BMM_Additive
    translationSpace=BCS_WorldSpace
    alphaBoolBlend {blendInTime=1, blendOption=Linear, blendOutTime=1}
    boneToModify {boneName=upperarm_r}
   }
   Save cached pose 'StartPose'(Pose=Component To Local(ComponentPose=@n21)) {cacheName=StartPose}
   n22 = "Transform (Modify) Bone - Bone: dyn_hair_bk_l_MSTR"(ComponentPose=Local To Component(LocalPose=Use cached pose 'StartPose'()), Alpha=1, Rotation="0,0,-35") {
    rotationMode=BMM_Additive
    rotationSpace=BCS_BoneSpace
    scaleMode=BMM_Additive
    translationSpace=BCS_WorldSpace
    alphaBoolBlend {blendInTime=1, blendOption=Linear, blendOutTime=1}
    boneToModify {boneName=dyn_hair_bk_l_MSTR}
    rotation {roll=-35}
   }
   n23 = "Transform (Modify) Bone - Bone: dyn_hair_bk_MSTR"(ComponentPose=@n22, Alpha=1, Rotation="0,0,-5") {
    rotationMode=BMM_Additive
    rotationSpace=BCS_BoneSpace
    scaleMode=BMM_Additive
    translationSpace=BCS_WorldSpace
    alphaBoolBlend {blendInTime=1, blendOption=Linear, blendOutTime=1}
    boneToModify {boneName=dyn_hair_bk_MSTR}
    rotation {roll=-5}
   }
   n24 = "Transform (Modify) Bone - Bone: dyn_hair_bk_r_MSTR"(ComponentPose=@n23, Alpha=1, Rotation="0,0,-35") {
    rotationMode=BMM_Additive
    rotationSpace=BCS_BoneSpace
    scaleMode=BMM_Additive
    translationSpace=BCS_WorldSpace
    alphaBoolBlend {blendInTime=1, blendOption=Linear, blendOutTime=1}
    boneToModify {boneName=dyn_hair_bk_r_MSTR}
    rotation {roll=-35}
   }
   Save cached pose 'ParachutePose'(Pose=Component To Local(ComponentPose=@n24)) {cacheName=ParachutePose}
   n25 = RigidBody!(
    ComponentPose=Local To Component(LocalPose=Use cached pose 'StartPose'())
    Alpha:float
    OverrideWorldGravity="0,0,-980"
    ComponentLinearAccScale="0.35,0.35,0.35"
    ComponentLinearVelScale="0.2,0.2,0.2"
   ) -> (Pose) {
    bClampLinearTranslationLimitToRefPose=true
    bOverrideWorldGravity=true
    overridePhysicsAsset=&/F_MED_TimeSquare_Hair_FaceAcc_BackHairPhysics
    simulationSpace=BaseBoneSpace
    alphaBoolBlend {blendOption=Linear}
    alphaScaleBiasClamp {interpSpeedDecreasing=4, interpSpeedIncreasing=6}
    baseBoneRef {boneName=spine_05}
    componentAppliedLinearAccClamp=<3000 3000 3000>
    componentLinearAccScale=<0.4 0.4 0.4>
    componentLinearVelScale=<0.3 0.3 0.3>
    overrideWorldGravity {z=-875}
    simSpaceSettings {maxAngularAcceleration=200, maxAngularVelocity=5, maxLinearAcceleration=500, maxLinearVelocity=200}
   }
   n26 = "Transform (Modify) Bone - Bone: dyn_hair_l_MSTR"!(ComponentPose=Local To Component(LocalPose=Use cached pose 'StartPose'()), Alpha=1, Translation="0,-2,1", Rotation="-0.000,0.000,30.000") -> (Pose) {
    rotationMode=BMM_Additive
    rotationSpace=BCS_BoneSpace
    scaleMode=BMM_Additive
    translationMode=BMM_Additive
    translationSpace=BCS_BoneSpace
    alphaBoolBlend {blendInTime=1, blendOption=Linear, blendOutTime=1}
    boneToModify {boneName=dyn_hair_l_MSTR}
    rotation {roll=30}
    translation {y=-2, z=1}
   }
   n27 = "Transform (Modify) Bone - Bone: dyn_hair_r_MSTR"!(ComponentPose=@n26, Alpha=1, Translation="0,-2,1", Rotation="0,0,30") -> (Pose) {
    rotationMode=BMM_Additive
    rotationSpace=BCS_BoneSpace
    scaleMode=BMM_Additive
    translationMode=BMM_Additive
    translationSpace=BCS_BoneSpace
    alphaBoolBlend {blendInTime=1, blendOption=Linear, blendOutTime=1}
    boneToModify {boneName=dyn_hair_r_MSTR}
    rotation {roll=30}
    translation {y=-2, z=1}
   }
   n28 = Blend Poses by bool(
    BlendPose_0=Use cached pose 'DiveModePose'()
    BlendPose_1=Component To Local(ComponentPose=@n27)
    BlendTime_0=0.9
    BlendTime_1=0.9
    bActiveValue=Get bIsSkydiveDiveMode()
   ) {blendType=Linear}
   n29 = Use cached pose 'ParachutePose'()
   n30 = Get bIsParachuteOpened()
   n31 = RigidBody(
    ComponentPose=Local To Component(LocalPose=Blend Poses by bool(BlendPose_0=@n29, BlendPose_1=@n28, BlendTime_0=0.9, BlendTime_1=0.9, bActiveValue=@n30) {blendType=Linear})
    OverrideWorldGravity="0,0,-980"
    ComponentLinearAccScale="0.38,0.38,0.38"
    ComponentLinearVelScale="0.2,0.2,0.2"
   ) {
    bClampLinearTranslationLimitToRefPose=true
    bOverrideWorldGravity=true
    overridePhysicsAsset=&/F_MED_TimeSquare_Hair_FaceAcc_BackHairPhysics_forSky
    simulationSpace=BaseBoneSpace
    alphaBoolBlend {blendOption=Linear}
    alphaScaleBiasClamp {interpSpeedDecreasing=4, interpSpeedIncreasing=6}
    baseBoneRef {boneName=spine_05}
    componentAppliedLinearAccClamp=<3000 3000 3000>
    componentLinearAccScale=<0.4 0.4 0.4>
    componentLinearVelScale=<0.3 0.3 0.3>
    overrideWorldGravity {z=-875}
    simSpaceSettings {maxAngularAcceleration=200, maxAngularVelocity=5, maxLinearAcceleration=500, maxLinearVelocity=200}
   }
   Save cached pose 'RB_BackPose'(Pose=Blend Poses by bool(BlendPose_0=Component To Local(ComponentPose=@n31), BlendPose_1=Component To Local(ComponentPose=@n25), bActiveValue=Get bIsSkyDiving()) {blendType=Linear}) {cacheName=RB_BackPose}
  ]}
 EventGraph {nodes [
   Event BlueprintUpdateAnimation()
   > n1 = Sequence()
   > Set WearingBack(WearingBack:bool=CheckForBackbling()) -> (Output_Get:bool)
   n2 = Gravity Override State Function DEPRECATED(
    execute=@n1.then_1
    CharacterPartAnimInstance=Self-Reference()
    GameTimeInSeconds=GetGameTimeInSeconds()
    GravityOverrideParams=Get Gravity Override Params()
   )
   > Set Gravity_Override(Gravity_Override:Vector=@n2) -> (Output_Get:Vector)
   n3 = Get BodyMesh() -> (BodyMesh:Skeletal Mesh Component Object Reference)
   > Is Valid(InputObject=@n3)
   > Set Upperarm_L_Rotation(Upperarm_L_Rotation:Rotator=BreakTransform(InTransform=GetSocketTransform(self=@n3, InSocketName=upperarm_l, TransformSpace=RTS_World)).Rotation) -> (Output_Get:Rotator)
   > Set Upperarm_R_Rotation(Upperarm_R_Rotation:Rotator=BreakTransform(InTransform=GetSocketTransform(self=@n3, InSocketName=upperarm_r, TransformSpace=RTS_World)).Rotation) -> (Output_Get:Rotator)
   n4 = Event OnCharacterCustomizationLoaded()
   > Set BodyMesh(BodyMesh=Get Array Item with EFort Custom Part Type(Array=@n4.CustomMeshParts))
  ]}
]
variables [
 BodyMesh {type=SkeletalMeshComponent}
 Gravity Override Params {type=GravityOverrideParamsStruct}
 Gravity_Override {type=Vector}
 Upperarm_L_Rotation {type=Rotator}
 Upperarm_R_Rotation {type=Rotator}
 WearingBack {type=boolean}
]
